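; Custom sprint movement mode extension for LemonExam that ties sprint to the equipped weapon, so that using sprint deducts its durability cost.
; On activation, the dedicated server gets the item-activated runtime data from the context and checks the owner pawn for the weapon-equipped tag.
; With the tag present it sets the activating item from the pawn's current weapon; without it, it writes a shipping log error.
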(event EventOnExtensionActivate
  (params (RuntimeContext :type MMERuntime%20Context%20Structure%20%28by%20ref%29))
  ;[0.000242 0.08174 0.15 0.5] OnExtensionActivate: Update the activating item so that durability cost is reducted when sprint is used.
  (bind _ownerpawn (Variables|FortMovementMode_BaseExtLogic|GetOwnerPawn)
    (:then
      (if (Networking|IsDedicatedServer _ownerpawn)
        (bind _returnvalue (Class|FortMovementModeExtBlueprintLibrary|GetRuntimeDatafromContext RuntimeContext "/Script/FortniteGame.FortMovementMode_ItemActivatedRuntimeData"))
        (Utilities|IsValid _returnvalue
          (:Is%20Valid
            (if (GameplayTags|HasMatchingGameplayTag _ownerpawn (Variables|Default|GetWeaponEquippedTag))
              (Class|FortMovementModeItemActivatedRuntimeData|SetActivatingItemfromWeapon _returnvalue (Variables|/Script/FortniteGame.FortPawn|GetCurrentWeapon _ownerpawn))
              (else
                (Fortnite|FortShippingLog self "MME_LemonExam_CustomSprint activated but the Equipped tag is missing from the OwnerPawn! Cannot set activating item from weapon reliably." true)))))))
    (:else)))
